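# Editor utility widget in the Art Review SunTool that orbits the level viewport camera around a focus actor.
# The focus is the selected actor or, with auto-lock on, the PIE player pawn found when slider dragging begins.
# SetTargetData caches the target location and camera offset and turns the camera toward it.
# The slider rotates that cached offset around world Z and adds matching yaw to the camera.
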EditorUtilityWidgetBlueprint /Game/ArtTools/Tech_Art/ArtReview_SunTool/SupportWidgets/ArtReview_CameraControls
graphs [
 EventGraph {nodes [
   n1 = "On Selection Changed (Button_SelectPlayerPawn)"!() -> (OutputDelegate, then, selected:bool, widget:User Widget Object Reference) {is=title K2Node_ComponentBoundEvent}
   > Set IsFocusingPlayerPawn?(IsFocusingPlayerPawn?:bool=@n1.selected) -> (Output_Get:bool)
   > Set HasValidTarget?(HasValidTarget?:bool=NOT Boolean(A=@n1.selected)) -> (Output_Get:bool)
   n2 = "On Value Changed (Slider_RotateAroundSelection)"!() -> (OutputDelegate, then, Value:float) {is=title K2Node_ComponentBoundEvent}
   > n3 = Branch(Condition=Get HasValidTarget?() -> (HasValidTarget?:bool))
   n4 = vector * integer!(A:Vector=Get LookAtTargetVector() -> (LookAtTargetVector:Vector), B:int=-1) -> (ReturnValue:Vector)
   SetLevelViewportCameraInfo!(
    execute=[@n3.then, @n12.then]
    self:Unreal Editor Subsystem Object Reference=Get UnrealEditorSubsystem()
    CameraLocation:Vector=vector + vector(
     A:Vector=Rotate Vector Around Axis!(InVect:Vector=@n4, AngleDeg:float=@n2.Value, Axis:Vector="0, 0,1") -> (ReturnValue:Vector)
     B:Vector=Get TargetActorLocation() -> (TargetActorLocation:Vector)
    ) -> (ReturnValue:Vector) {is=title K2Node_CommutativeAssociativeBinaryOperator}
    CameraRotation:Rotator=Combine Rotators!(A:Rotator=Get LookAtTargetRotator() -> (LookAtTargetRotator:Rotator), B_Roll:float=0, B_Pitch:float=0, B_Yaw:float=@n2.Value) -> (ReturnValue:Rotator)
   ) -> (then)
   "On Clicked (Button_SelectFocus)"!() -> (OutputDelegate, then, widget:User Widget Object Reference, event:Pointer Event Structure) {is=title K2Node_ComponentBoundEvent}
   > n5 = GetSelectedActors()
   > n6 = Branch(Condition=Is Array Non Empty!(InputArray:Array of Actor Object References=@n5) -> (Non-Empty:bool=~) {is=title K2Node_MacroInstance})
   > setSelected!(
    execute
    self:TA Base Editor Button Widget Object Reference=Get Button_SelectPlayerPawn() -> (Button_SelectPlayerPawn:TA Generic 01 Button Object Reference)
    selected:bool=false
    throwEvent:bool=true
   ) -> (then)
   n7 = Get Slider_RotateAroundSelection() -> (Slider_RotateAroundSelection:Slider Object Reference)
   > Set Enabled And Tooltip Text!(execute, Widget:Widget Object Reference=@n7, Enabled:bool=true, Tooltip Text:Text, __WorldContext:Object Reference) -> (then) {is=title K2Node_CallFunction}
   > SetValue!(execute, self:Slider Object Reference=@n7, InValue:float=0) -> (then)
   > n8 = GetSelectedActors()
   > Set HasValidTarget?(HasValidTarget?=true)
   > SetTargetData!(
    execute
    self
    TargetActor:Actor Object Reference="Get (a copy)"(Array:Array of Actor Object References=@n8) -> (Output:Actor Object Reference)
   ) -> (then)
   PrintString(execute=@n6.else, InString=Error - No Selected Actor)
   "On Mouse Capture Begin (Slider_RotateAroundSelection)"() {is=title K2Node_ComponentBoundEvent}
   > Branch(Condition=Get IsFocusingPlayerPawn?() -> (IsFocusingPlayerPawn?:bool))
   > n9 = GetPIEWorlds!(execute, bIncludeDedicatedServer:bool=false) -> (then, ReturnValue:Array of World Object References)
   > n10 = Branch(Condition=Is Array Non Empty(InputArray:Array of World Object References=@n9) -> (Non-Empty=~))
   > n11 = GetAllActorsOfClass!(
    execute
    WorldContextObject:Object Reference="Get (a copy)"(Array:Array of World Object References=@n9) -> (Output:World Object Reference)
    ActorClass:Actor Class Reference=/Game/Athena/PlayerPawn_Athena.PlayerPawn_Athena_C
   ) -> (then, OutActors:Array of Player Pawn Athena Object References)
   > SetTargetData(TargetActor="Get (a copy)"(Array:Array of Player Pawn Athena Object References=@n11) -> (Output:Player Pawn Athena Object Reference))
   > n12 = Set HasValidTarget?(HasValidTarget?=true)
   PrintString(execute=@n10.else, InString=Auto Select Only Works in PIE. Please select asset for Focus.)
   Set Enabled And Tooltip Text(Widget=Get Slider_RotateAroundSelection(), Enabled=false, Tooltip Text="LOCTEXT(\"Focus Asset to Enable\")")
   PrintString(InString=PAth)
   Event PreConstruct()
   Event Construct()
   Event Tick()
   Event OnRemovedFromFocusPath!() -> (OutputDelegate, then, InFocusEvent:Focus Event Structure) {is=title K2Node_Event}
  ]}
 OnKeyDown {nodes [
   OnKeyDown!() -> (then, MyGeometry:Geometry Structure, InKeyEvent:Key Event Structure) {is=title K2Node_FunctionEntry}
   > PrintString(InString=Hello)
   > Return Node!(execute, ReturnValue:Event Reply Structure=Unhandled())
  ]}
 OnMouseWheel {nodes [
   n1 = OnMouseWheel!() -> (then, MyGeometry:Geometry Structure, MouseEvent:"Pointer Event Structure (by ref)") {is=title K2Node_FunctionEntry}
   n2 = GetInputEventFromPointerEvent!(Event:"Pointer Event Structure (by ref)"=@n1.MouseEvent) -> (ReturnValue:Input Event Structure)
   > PrintString(InString="To String (Boolean)"(InBool=Is Shift Down!(Input:"Input Event Structure (by ref)"=@n2) -> (ReturnValue:bool)))
   > Return Node(ReturnValue=Unhandled())
   Get Wheel Delta!(Input:"Pointer Event Structure (by ref)"=@n1.MouseEvent) -> (ReturnValue:float) {is=title K2Node_CallFunction}
  ]}
 OnPaint {nodes [
   n1 = OnPaint!() -> (then, Context:"Paint Context Structure (by ref)") {is=title K2Node_FunctionEntry}
   > Create SliderLabel!(
    execute
    OnPaintContext:Paint Context Structure=@n1
    Slider:Slider Object Reference=Get Slider_RotateAroundSelection()
    PositionalShift:Vector 2D Structure=(X=5,Y=13)
    Label:String=Rotate Camera
    Label Color:Linear Color Structure=(R=0.9249,G=1,B=0.9578,A=1)
    FontSize:int=9
    __WorldContext:Object Reference
   ) -> (then, ContextOut:Paint Context Structure)
   > Return Node!(execute)
  ]}
 SetTargetData {nodes [
   n1 = SetTargetData!() -> (then, TargetActor:Actor Object Reference) {is=title K2Node_FunctionEntry}
   n2 = Get Actor Location(self=@n1)
   > Set TargetActorLocation(TargetActorLocation:Vector=@n2) -> (Output_Get:Vector)
   > n3 = Get Editor Camera Info!(execute) -> (then, CameraLocation:Vector, ForwardDirection:Vector, FOV:float, ReturnValue:bool) {is=title K2Node_CallFunction}
   > n4 = Sequence!(execute) -> (then_0, then_1, then_2:Exec, then_3:Exec)
   > Set CameraLocation(CameraLocation:Vector=@n3) -> (Output_Get:Vector)
   n5 = vector - vector!(A:Vector=@n2, B:Vector=@n3) -> (ReturnValue:Vector) {is=title K2Node_CallFunction}
   Set LookAtTargetVector(execute=@n4.then_1, LookAtTargetVector:Vector=@n5) -> (Output_Get:Vector)
   n6 = Rotation From X Vector!(InVec:Vector=@n5) -> (ReturnValue:Rotator)
   Set LookAtTargetRotator(execute=@n4.then_2, LookAtTargetRotator:Rotator=@n6) -> (Output_Get:Rotator)
   SetLevelViewportCameraInfo(execute=@n4.then_3, self=Get UnrealEditorSubsystem(), CameraLocation=@n3, CameraRotation=@n6)
  ]}
]
variables [
 CameraLocation {type=Vector}
 HasValidTarget? {type=boolean}
 IsFocusingPlayerPawn? {type=boolean, default=true}
 LookAtTargetRotator {type=Rotator}
 LookAtTargetVector {type=Vector}
 TargetActorLocation {type=Vector}
]
widgets {
 info {rootWidgetClass=&Overlay, widgetCount=6}
 widgets [
  Overlay_27 {widgetClassPath=&Overlay, widget=Overlay_27}
  HorizontalBox_0 {
   widgetClassPath=&HorizontalBox
   parent=Overlay_27
   widget=HorizontalBox_0
   slot {class=OverlaySlot, properties {horizontalAlignment=HAlign_Fill}}
  }
  VerticalBox_0 {
   widgetClassPath=&VerticalBox
   parent=HorizontalBox_0
   widget=VerticalBox_0
   slot {class=HorizontalBoxSlot, properties {size {sizeRule=Fill, value=0.15}}}
  }
  Button_SelectFocus {
   widgetClassPath=&TA_Generic01_Button
   parent=VerticalBox_0
   widget=Button_SelectFocus
   properties {label=Set Selection As Focus}
   slot {class=VerticalBoxSlot}
  }
  Button_SelectPlayerPawn {
   widgetClassPath=&TA_Generic01_Button
   parent=VerticalBox_0
   widget=Button_SelectPlayerPawn
   properties {
    autoToggleSelection=true
    isEnabled=false
    isSelected=true
    label=Auto Lock Player Pawn
    buttonColor_Default=<rgba 1 0 0 1>
    buttonColor_Selected=<rgba 0 1 0 1>
   }
   slot {class=VerticalBoxSlot}
  }
  Slider_RotateAroundSelection {
   widgetClassPath=/Script/UMG.Slider
   bIsVariable=true
   parent=HorizontalBox_0
   widget=Slider_RotateAroundSelection
   properties {
    maxValue=210
    minValue=-210
    widgetStyle {
     barThickness=4
     disabledBarImage {
      bIsDynamicallyLoaded=false
      drawAs=RoundedBox
      imageType=NoImage
      mirroring=NoMirror
      resourceName=None
      resourceObject=null
      tiling=NoTile
      imageSize=<32 32>
      margin=<ltrb 0 0 0 0>
      outlineSettings {
       bUseBrushTransparency=false
       roundingType=FixedRadius
       width=1
       color {colorUseRule=UseColor_ColorTable, specifiedColor=<rgba 0 0 0 0>}
       cornerRadii=<2 2 2 2>
      }
      tintColor {colorUseRule=UseColor_Specified, specifiedColor=<rgba 1 1 1 1>}
      uVRegion {bIsValid=false, max=<0 0>, min=<0 0>}
     }
     disabledThumbImage {
      bIsDynamicallyLoaded=false
      drawAs=RoundedBox
      imageType=NoImage
      mirroring=NoMirror
      resourceName=None
      resourceObject=null
      tiling=NoTile
      imageSize=<32 32>
      margin=<ltrb 0 0 0 0>
      outlineSettings {
       bUseBrushTransparency=false
       roundingType=HalfHeightRadius
       width=0
       color {colorUseRule=UseColor_Specified, specifiedColor=<rgba 0 0 0 0>}
       cornerRadii=<0 0 0 0>
      }
      tintColor {colorUseRule=UseColor_Specified, specifiedColor=<rgba 0.5271 0.5271 0.5271 1>}
      uVRegion {bIsValid=false, max=<0 0>, min=<0 0>}
     }
     hoveredBarImage {
      bIsDynamicallyLoaded=false
      drawAs=RoundedBox
      imageType=NoImage
      mirroring=NoMirror
      resourceName=None
      resourceObject=null
      tiling=NoTile
      imageSize=<0 0>
      margin=<ltrb 0 0 0 0>
      outlineSettings {
       bUseBrushTransparency=false
       roundingType=FixedRadius
       width=1
       color {colorUseRule=UseColor_ColorTable, specifiedColor=<rgba 0 0 0 0>}
       cornerRadii=<2 2 2 2>
      }
      tintColor {colorUseRule=UseColor_Specified, specifiedColor=<rgba 0.0048 0.0048 0.0048 1>}
      uVRegion {bIsValid=false, max=<0 0>, min=<0 0>}
     }
     hoveredThumbImage {
      bIsDynamicallyLoaded=false
      drawAs=RoundedBox
      imageType=NoImage
      mirroring=NoMirror
      resourceName=None
      resourceObject=null
      tiling=NoTile
      imageSize=<8 8>
      margin=<ltrb 0 0 0 0>
      outlineSettings {
       bUseBrushTransparency=false
       roundingType=HalfHeightRadius
       width=0
       color {colorUseRule=UseColor_Specified, specifiedColor=<rgba 0 0 0 0>}
       cornerRadii=<0 0 0 0>
      }
      tintColor {colorUseRule=UseColor_Specified, specifiedColor=<rgba 1 1 1 1>}
      uVRegion {bIsValid=false, max=<0 0>, min=<0 0>}
     }
     normalBarImage {
      bIsDynamicallyLoaded=false
      drawAs=RoundedBox
      imageType=NoImage
      mirroring=NoMirror
      resourceName=None
      resourceObject=null
      tiling=NoTile
      imageSize=<0 0>
      margin=<ltrb 0 0 0 0>
      outlineSettings {
       bUseBrushTransparency=false
       roundingType=FixedRadius
       width=1
       color {colorUseRule=UseColor_ColorTable, specifiedColor=<rgba 0 0 0 0>}
       cornerRadii=<2 2 2 2>
      }
      tintColor {colorUseRule=UseColor_Specified, specifiedColor=<rgba 0.0048 0.0048 0.0048 1>}
      uVRegion {bIsValid=false, max=<0 0>, min=<0 0>}
     }
     normalThumbImage {
      bIsDynamicallyLoaded=false
      drawAs=RoundedBox
      imageType=NoImage
      mirroring=NoMirror
      resourceName=None
      resourceObject=null
      tiling=NoTile
      imageSize=<8 8>
      margin=<ltrb 0 0 0 0>
      outlineSettings {
       bUseBrushTransparency=false
       roundingType=HalfHeightRadius
       width=0
       color {colorUseRule=UseColor_Specified, specifiedColor=<rgba 0 0 0 0>}
       cornerRadii=<0 0 0 0>
      }
      tintColor {colorUseRule=UseColor_Specified, specifiedColor=<rgba 0.2159 0.2159 0.2159 1>}
      uVRegion {bIsValid=false, max=<0 0>, min=<0 0>}
     }
    }
   }
   slot {class=HorizontalBoxSlot, properties {size {sizeRule=Fill, value=1}}}
  }
 ]
}
