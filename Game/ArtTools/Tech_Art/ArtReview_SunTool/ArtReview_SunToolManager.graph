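# Editor utility widget for art review, hosting sun, camera and visualizer sub-widgets.
# Setup, blocked during PIE, ensures a BP_ArtReviewHelper_Chris exists and checks that World Settings use BR_Art_ReviewTODMSystem.
# View-pass buttons, in a collapsed panel, call that helper, which FindTODMHelperBP finds or creates, preferring PIE worlds.
# It also toggles vertex color via console command, disables Slate throttling on construct, and has functions switching suns through the helper.
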EditorUtilityWidgetBlueprint /Game/ArtTools/Tech_Art/ArtReview_SunTool/ArtReview_SunToolManager
graphs [
 EventGraph {nodes [
   "On Clicked (Button_BaseColor)"!() -> (OutputDelegate, then, widget:User Widget Object Reference, event:Pointer Event Structure) {is=title K2Node_ComponentBoundEvent}
   > n1 = FindTODMHelperBP!(execute, self, bIncludeDedicatedServer:bool=false) -> (then, Output_Get:BP Art Review Helper Chris Object Reference)
   > BaseColor(self:BP Art Review Helper Chris Object Reference=@n1)
   "On Clicked (Button_Metallic)"!() -> (OutputDelegate, then, widget:User Widget Object Reference, event:Pointer Event Structure) {is=title K2Node_ComponentBoundEvent}
   > n2 = FindTODMHelperBP(bIncludeDedicatedServer=false)
   > Metallic(self:BP Art Review Helper Chris Object Reference=@n2)
   "On Clicked (Button_Roughness)"!() -> (OutputDelegate, then, widget:User Widget Object Reference, event:Pointer Event Structure) {is=title K2Node_ComponentBoundEvent}
   > n3 = FindTODMHelperBP(bIncludeDedicatedServer=false)
   > Roughness(self:BP Art Review Helper Chris Object Reference=@n3)
   "On Clicked (Button_Specular)"!() -> (OutputDelegate, then, widget:User Widget Object Reference, event:Pointer Event Structure) {is=title K2Node_ComponentBoundEvent}
   > n4 = FindTODMHelperBP(bIncludeDedicatedServer=false)
   > Specular(self:BP Art Review Helper Chris Object Reference=@n4)
   "On Clicked (Button_Emissive)"!() -> (OutputDelegate, then, widget:User Widget Object Reference, event:Pointer Event Structure) {is=title K2Node_ComponentBoundEvent}
   > n5 = FindTODMHelperBP(bIncludeDedicatedServer=false)
   > Emissive(self:BP Art Review Helper Chris Object Reference=@n5)
   ResetButtonOnClickedAction() {is=title K2Node_CustomEvent}
   > n6 = FindTODMHelperBP(bIncludeDedicatedServer=false)
   > Reset(self:BP Art Review Helper Chris Object Reference=@n6)
   Event Construct()
   > ExecuteConsoleCommand(Command=Slate.bAllowThrottling 0)
   Event PreConstruct()
   Event Tick()
  ]}
 FindTODMHelperBP {nodes [
   n1 = FindTODMHelperBP!() -> (then, bIncludeDedicatedServer:bool=~) {is=title K2Node_FunctionEntry}
   > n2 = GetPIEWorlds!(execute, bIncludeDedicatedServer:bool=@n1) -> (then, ReturnValue:Array of World Object References)
   > n3 = Branch(Condition=Is Array Non Empty!(InputArray:Array of World Object References=@n2) -> (Non-Empty:bool=~) {is=title K2Node_MacroInstance})
   > n4 = For Each Loop(Array:Array of World Object References=@n2) -> (Array Element:World Object Reference)
   > n5 = GetAllActorsOfClass!(
    execute
    WorldContextObject:Object Reference=@n4
    ActorClass:Actor Class Reference=/Game/ArtTools/Tech_Art/ArtReview_SunTool/SupportBPs/BP_ArtReviewHelper_Chris.BP_ArtReviewHelper_Chris_C
   ) -> (then, OutActors:Array of BP Art Review Helper Chris Object References)
   > n6 = Branch(Condition=Is Array Non Empty(InputArray:Array of BP Art Review Helper Chris Object References=@n5) -> (Non-Empty=~))
   > n7 = Set TODMHelperRef(TODMHelperRef:BP Art Review Helper Chris Object Reference="Get (a copy)"(Array:Array of BP Art Review Helper Chris Object References=@n5) -> (Output:BP Art Review Helper Chris Object Reference)) -> (Output_Get:BP Art Review Helper Chris Object Reference)
   > Return Node!(execute, Output_Get:BP Art Review Helper Chris Object Reference=@n7)
   n8 = SpawnActor BP Art Review Helper Chris!(
    execute:Exec=@n6.else
    WorldContextObject:Object Reference=@n4
    Class:Actor Class Reference=&BP_ArtReviewHelper_Chris_C
    SpawnTransform_Location:Vector="0, 0, 0"
    SpawnTransform_Rotation:Rotator="0, 0, 0"
    SpawnTransform_Scale:Vector="1,1,1"
    CollisionHandlingOverride:ESpawnActorCollisionHandlingMethod Enum=Undefined
    TransformScaleMethod:ESpawnActorScaleMethod Enum=OverrideRootScale
    Owner:Actor Object Reference
    Instigator:Pawn Object Reference
   ) -> (then:Exec, ReturnValue:BP Art Review Helper Chris Object Reference) {is=title K2Node_SpawnActorFromClass}
   > n9 = Set TODMHelperRef(TODMHelperRef=@n8)
   > Return Node(Output_Get=@n9)
   n10 = GetAllActorsOfClass(execute=@n3.else, ActorClass=&BP_ArtReviewHelper_Chris_C)
   > n11 = Branch(Condition=Is Array Non Empty(InputArray=@n10) -> (Non-Empty=~))
   > n12 = Set TODMHelperRef(TODMHelperRef="Get (a copy)"(Array=@n10))
   > Return Node(Output_Get=@n12)
   QuickSetupForNonPie(execute=@n11.else)
   > n13 = GetAllActorsOfClass(ActorClass=&BP_ArtReviewHelper_Chris_C)
   > n14 = Set TODMHelperRef(TODMHelperRef="Get (a copy)"(Array=@n13))
   > Return Node(Output_Get=@n14)
  ]}
 ManageButtonSelection {nodes [
   PrintString(InString=Append(A="Widget Switch: ", B="To String (Integer)"(InInt=0)))
   > n1 = Switch on Int!(execute:Exec, Selection:int=0) -> (Default:Exec, 0:Exec, 1:Exec)
   > PrintString(InString=Error. Unexpected Tab, TextColor=(R=1,G=0.0072,B=0,A=1), Duration=10)
   n2 = FindTODMHelperBP(execute=@n1.1, bIncludeDedicatedServer=false) {is=title K2Node_CallFunction}
   > SwitchSuns!(execute, self, IsFalseSun:bool=true, TODMHelperRef:BP Art Review Helper Chris Object Reference=@n2) -> (then)
   n3 = FindTODMHelperBP(execute=@n1.0, bIncludeDedicatedServer=false)
   > SwitchSuns(IsFalseSun=false, TODMHelperRef=@n3)
   n4 = "On Selection Changed (Button_BaseColor)"!() -> (OutputDelegate, then, selected:bool, widget:User Widget Object Reference) {is=title K2Node_ComponentBoundEvent}
   > Toggle Selection Of View Pass!(execute:Exec, Condition:bool=@n4.selected, Widget:User Widget Object Reference=@n4.widget) {is=title K2Node_MacroInstance}
   n5 = "On Selection Changed (Button_Emissive)"!() -> (OutputDelegate, then, selected:bool, widget:User Widget Object Reference) {is=title K2Node_ComponentBoundEvent}
   > Toggle Selection Of View Pass(Condition=@n5.selected, Widget=@n5.widget)
   n6 = "On Selection Changed (Button_Metallic)"!() -> (OutputDelegate, then, selected:bool, widget:User Widget Object Reference) {is=title K2Node_ComponentBoundEvent}
   > Toggle Selection Of View Pass(Condition=@n6.selected, Widget=@n6.widget)
   n7 = "On Selection Changed (Button_Roughness)"!() -> (OutputDelegate, then, selected:bool, widget:User Widget Object Reference) {is=title K2Node_ComponentBoundEvent}
   > Toggle Selection Of View Pass(Condition=@n7.selected, Widget=@n7.widget)
   n8 = "On Selection Changed (Button_Specular)"!() -> (OutputDelegate, then, selected:bool, widget:User Widget Object Reference) {is=title K2Node_ComponentBoundEvent}
   > Toggle Selection Of View Pass(Condition=@n8.selected, Widget=@n8.widget)
   n9 = "On Clicked (Button_Reset)"!() -> (OutputDelegate, then, widget:User Widget Object Reference, event:Pointer Event Structure) {is=title K2Node_ComponentBoundEvent}
   ClearPreviousSelection(execute=[@n9.then, @n34.then])
   > ResetButtonOnClickedAction() {is=title K2Node_CallFunction}
   n10 = "On Clicked (Button_Setup)"!() -> (OutputDelegate, then, widget:User Widget Object Reference, event:Pointer Event Structure) {is=title K2Node_ComponentBoundEvent}
   n11 = GetPIEWorlds(execute=[@n10.then, @n33.then], bIncludeDedicatedServer=false)
   > n12 = Branch(Condition=Is Array Non Empty(InputArray:Array of World Object References=@n11) -> (Non-Empty=~))
   > PrintString(InString=Error: Please do not attempt setup while running PIE, TextColor=(R=1,G=0.0702,B=0.0304,A=1), Duration=5)
   n13 = GetActorOfClass!(
    execute=@n12.else
    WorldContextObject:Object Reference
    ActorClass:Actor Class Reference=&BP_ArtReviewHelper_Chris_C
   ) -> (then, ReturnValue:BP Art Review Helper Chris Object Reference)
   > n14 = Is Valid(InputObject=@n13)
   PrintString(execute=@n14.Is Not Valid, InString="No Helper BP, Adding to world")
   > n15 = GetEditorWorld(self=Get UnrealEditorSubsystem())
   > SpawnActor BP Art Review Helper Chris(
    WorldContextObject=@n15
    Class=&BP_ArtReviewHelper_Chris_C
    SpawnTransform_Location="0, 0, 0"
    SpawnTransform_Rotation="0, 0, 0"
    SpawnTransform_Scale="1,1,1"
    CollisionHandlingOverride=Undefined
    TransformScaleMethod=OverrideRootScale
   )
   > n16 = Set IsBPHelperInWorld?(IsBPHelperInWorld?:bool=true) -> (Output_Get:bool)
   n17 = GetEditorWorld(execute=[@n27.then, @n16.then], self=Get UnrealEditorSubsystem())
   > n18 = GetWorldSettings!(execute, self:World Object Reference=@n17) -> (then, ReturnValue:World Settings Object Reference)
   > n19 = Cast To FortWorldSettings!(execute, Object=@n18) -> (then, CastFailed, AsFort World Settings:Fort World Settings Object Reference) {is=title K2Node_DynamicCast}
   > n20 = GetTimeOfDayManager!(execute, self:Fort World Settings Object Reference=@n19) -> (then, ReturnValue:Fort Time Of Day Manager Interface Interface)
   n21 = "To Object (Interface)"!(Interface:"Interface Interface (by ref)"=@n20) -> (ReturnValue:Object Reference)
   > n22 = Branch(Condition="Equal (Class)"!(
     A:Object Class Reference=Get Class(Object=@n21) -> (ReturnValue:Object Class Reference)
     B:Object Class Reference=/Game/ArtTools/Tech_Art/ArtReview_SunTool/SupportBPs/BR_Art_ReviewTODMSystem.BR_Art_ReviewTODMSystem_C
    ) -> (ReturnValue:bool) {is=title K2Node_CallFunction})
   > PrintString(InString=Setup Complete)
   > n23 = Set IsTODMSetCorrectly?(IsTODMSetCorrectly?:bool=true) -> (Output_Get:bool)
   n24 = Branch(execute=[@n26.then, @n23.then], Condition=Get IsTODMSetCorrectly?() -> (IsTODMSetCorrectly?:bool))
   n25 = Get Button_Setup() -> (Button_Setup:TA Generic 01 Button Object Reference)
   > Set buttonColor_Default!(execute, buttonColor_Default:Linear Color Structure=(R=0.0043,G=1,B=0,A=1), self:TA Base Editor Button Widget Object Reference=@n25) -> (then, Output_Get:Linear Color Structure)
   > setLabel!(execute, self:TA Base Editor Button Widget Object Reference=@n25, label:String=Setup Complete) -> (then)
   Set buttonColor_Default(execute=@n24.else, buttonColor_Default=(R=1,G=0,B=0,A=1), self=@n25)
   > setLabel(self=@n25, label=Wrong TODM - BR_Art_ReviewTODMSystem)
   > SetToolTipText!(
    execute
    self:Widget Object Reference=@n25
    InToolTipText:"Text (by ref)"=MakeLiteralText(Value="LOCTEXT(\"Change World Settings - Time of Day - Time of Day Manager to BR_Art_ReviewTODMSystem\")")
   ) -> (then)
   PrintString(
    execute=@n22.else
    InString=Change World Settings - Time of Day - Time of Day Manager to BR_Art_ReviewTODMSystem
    TextColor=(R=1,G=0.0576,B=0,A=1)
    Duration=5
   )
   > n26 = Set IsTODMSetCorrectly?(IsTODMSetCorrectly?=false)
   PrintString(execute=@n19.CastFailed, InString="Error, World does not use Fort World Settings")
   n27 = PrintString(execute=@n14.Is Valid, InString=BP Helper Already Exists in World)
   n28 = "On Selection Changed (ToggleVertexColor)"!() -> (OutputDelegate, then, selected:bool, widget:User Widget Object Reference) {is=title K2Node_ComponentBoundEvent}
   > Clear Viewmodes()
   > n29 = Branch(Condition=@n28.selected)
   > ExecuteConsoleCommand(Command=ShowFlag.VertexColors 1)
   ExecuteConsoleCommand(execute=[@n35.then, @n29.else], Command=ShowFlag.VertexColors 0)
   > PrintString(InString=Hello)
   ClearPreviousSelection() {is=title K2Node_CustomEvent}
   n30 = Get IsCurrentlySelectedButton() -> (IsCurrentlySelectedButton:User Widget Object Reference)
   > Is Valid(InputObject=@n30)
   > n31 = Cast To TA_Generic01_Button!(execute, Object=@n30) -> (then, CastFailed, AsTA Generic 01 Button:TA Generic 01 Button Object Reference) {is=title K2Node_DynamicCast}
   > setSelected!(execute, self:TA Base Editor Button Widget Object Reference=@n31, selected:bool=false, throwEvent:bool=true) -> (then)
   PrintString(execute=@n31.CastFailed, InString=Something Went Wrong Attempting to Clear previously selected button)
   n32 = SetCurrentSelection!() -> (OutputDelegate, then, Widget:User Widget Object Reference) {is=title K2Node_CustomEvent}
   > Set IsCurrentlySelectedButton(IsCurrentlySelectedButton:User Widget Object Reference=@n32.Widget) -> (Output_Get:User Widget Object Reference)
   n33 = QuickSetupForNonPie() {is=title K2Node_CustomEvent}
   n34 = Clear Viewmodes() {is=title K2Node_CustomEvent}
   n35 = ClearVertexColor() {is=title K2Node_CustomEvent}
   Reset VertexColor Toggle() {is=title K2Node_CustomEvent}
   > ClearVertexColor() {is=title K2Node_CallFunction}
   > setSelected(self=Get ToggleVertexColor() -> (ToggleVertexColor:TA Generic 01 Button Object Reference), selected=false, throwEvent=true)
   n36 = Branch()
   > ExecuteConsoleCommand(Command=Slate.bAllowThrottling 0)
   n37 = Get Button_Setup()
   > setLabel(self=@n37, label=Pie Throttle OFF)
   ExecuteConsoleCommand(execute=@n36.else, Command=Slate.bAllowThrottling 1)
   > setLabel(self=@n37, label=Pie Throttle ON)
   Reset VertexColor Toggle() {is=title K2Node_CallFunction}
   "On Selection Changed (Button_Setup)"!() -> (OutputDelegate, then, selected:bool, widget:User Widget Object Reference) {is=title K2Node_ComponentBoundEvent}
  ]}
 SwitchSuns {nodes [
   n1 = GetActorOfClass(ActorClass=/Game/TimeOfDay/TODM/BR/s11/TODM_BR_s14.TODM_BR_s14_C) -> (ReturnValue:TODM BR S 14 Object Reference)
   > n2 = SetVisibility!(
    execute
    self:Scene Component Object Reference=Get DirectionalLightComponent!(self:Fort Time Of Day Manager Object Reference=@n1) -> (DirectionalLightComponent:Directional Light Component Object Reference)
    bNewVisibility:bool=true
    bPropagateToChildren:bool=false
   ) -> (then)
   PrintString(execute=[@n10.then, @n2.then], InString=TODM ON)
   n3 = GetActorOfClass(ActorClass=&TODM_BR_s14_C)
   > n4 = SetVisibility(self=Get DirectionalLightComponent(self=@n3), bNewVisibility=false, bPropagateToChildren=false)
   PrintString(execute=[@n9.then, @n4.then], InString=TODM OFF)
   n5 = SwitchSuns!() -> (then, IsFalseSun:bool=~, TODMHelperRef:BP Art Review Helper Chris Object Reference) {is=title K2Node_FunctionEntry}
   > n6 = Branch(Condition="Not Equal (Boolean)"!(A:bool=@n5, B:bool=Get State() -> (State:bool)) -> (ReturnValue:bool) {is=title K2Node_CallFunction})
   > n7 = Set State(State:bool=@n5) -> (Output_Get:bool)
   > n8 = Branch(Condition=@n7)
   > n9 = TODM_Off(self:BP Art Review Helper Chris Object Reference=@n5.TODMHelperRef)
   n10 = TODM_On(execute=@n8.else, self:BP Art Review Helper Chris Object Reference=@n5.TODMHelperRef)
   PrintString(execute=@n6.else, InString=Identical State in Switch Suns)
  ]}
 ToggleSelectionOfViewPass {nodes [
   n1 = Inputs!() -> (execute, Condition:bool=~, Widget:User Widget Object Reference)
   > Reset VertexColor Toggle()
   > Branch(Condition=@n1)
   > ClearPreviousSelection() {is=title K2Node_CallFunction}
   > SetCurrentSelection!(execute, self, Widget:User Widget Object Reference=@n1.Widget) -> (then) {is=title K2Node_CallFunction}
   Outputs()
  ]}
]
variables [
 IsBPHelperInWorld? {type=boolean}
 IsCurrentlySelectedButton {type=UserWidget, default=null}
 IsTODMSetCorrectly? {type=boolean}
 State {type=boolean, default=true}
 TODMHelperRef {type=BP_ArtReviewHelper_Chris_C, default=null}
]
widgets {
 info {rootWidgetClass=&Overlay, widgetCount=21}
 widgets [
  Overlay_25 {widgetClassPath=&Overlay, widget=Overlay_25}
  VerticalBox_2 {
   widgetClassPath=&VerticalBox
   parent=Overlay_25
   widget=VerticalBox_2
   slot {class=OverlaySlot, properties {horizontalAlignment=HAlign_Fill}}
  }
  HorizontalBox_60 {widgetClassPath=&HorizontalBox, parent=VerticalBox_2, widget=HorizontalBox_60, slot {class=VerticalBoxSlot}}
  VerticalBox_51 {
   widgetClassPath=&VerticalBox
   parent=HorizontalBox_60
   widget=VerticalBox_51
   slot {class=HorizontalBoxSlot, properties {size {sizeRule=Fill, value=0.15}}}
  }
  HorizontalBox_1 {widgetClassPath=&HorizontalBox, parent=VerticalBox_51, widget=HorizontalBox_1, slot {class=VerticalBoxSlot}}
  Button_Setup {
   widgetClassPath=&TA_Generic01_Button
   parent=HorizontalBox_1
   widget=Button_Setup
   properties {label=Setup Level for Tool, buttonColor_Default=<rgba 0.7719 0.7386 0 1>, buttonColor_Selected=<rgba 0.1104 0.5802 0 1>}
   slot {class=HorizontalBoxSlot, properties {size {sizeRule=Fill, value=1}}}
  }
  HorizontalBox_0 {
   widgetClassPath=&HorizontalBox
   parent=VerticalBox_51
   widget=HorizontalBox_0
   properties {visibility=Collapsed}
   slot {class=VerticalBoxSlot, properties {size {sizeRule=Fill, value=1}}}
  }
  VerticalBox_0 {
   widgetClassPath=&VerticalBox
   parent=HorizontalBox_0
   widget=VerticalBox_0
   slot {class=HorizontalBoxSlot, properties {size {sizeRule=Fill, value=1}}}
  }
  Button_Reset {
   widgetClassPath=&TA_Generic01_Button
   parent=VerticalBox_0
   widget=Button_Reset
   properties {label=Reset}
   slot {class=VerticalBoxSlot}
  }
  Button_BaseColor {
   widgetClassPath=&TA_Generic01_Button
   parent=VerticalBox_0
   widget=Button_BaseColor
   properties {autoToggleSelection=true, label=BaseColor}
   slot {class=VerticalBoxSlot, properties {size {sizeRule=Fill, value=1}}}
  }
  Button_Specular {
   widgetClassPath=&TA_Generic01_Button
   parent=VerticalBox_0
   widget=Button_Specular
   properties {autoToggleSelection=true, label=Specular}
   slot {class=VerticalBoxSlot, properties {size {sizeRule=Fill, value=1}}}
  }
  ToggleVertexColor {
   widgetClassPath=&TA_Generic01_Button
   parent=VerticalBox_0
   widget=ToggleVertexColor
   properties {autoToggleSelection=true, label=Toggle Vertex Color}
   slot {class=VerticalBoxSlot, properties {size {sizeRule=Fill, value=1}}}
  }
  VerticalBox_1 {
   widgetClassPath=&VerticalBox
   parent=HorizontalBox_0
   widget=VerticalBox_1
   slot {class=HorizontalBoxSlot, properties {size {sizeRule=Fill, value=1}}}
  }
  Button_Roughness {
   widgetClassPath=&TA_Generic01_Button
   parent=VerticalBox_1
   widget=Button_Roughness
   properties {autoToggleSelection=true, label=Roughness}
   slot {class=VerticalBoxSlot}
  }
  Button_Metallic {
   widgetClassPath=&TA_Generic01_Button
   parent=VerticalBox_1
   widget=Button_Metallic
   properties {autoToggleSelection=true, label=Metallic}
   slot {class=VerticalBoxSlot, properties {size {sizeRule=Fill, value=1}}}
  }
  Button_Emissive {
   widgetClassPath=&TA_Generic01_Button
   parent=VerticalBox_1
   widget=Button_Emissive
   properties {autoToggleSelection=true, label=Emissive}
   slot {class=VerticalBoxSlot, properties {size {sizeRule=Fill, value=1}}}
  }
  Spacer_130 {
   widgetClassPath=&Spacer
   parent=VerticalBox_1
   widget=Spacer_130
   slot {class=VerticalBoxSlot, properties {size {sizeRule=Fill, value=1}}}
  }
  ArtReview_PostProcess {
   widgetClassPath=/Game/NeverCook/ArtTools/Tech_Art/ArtReview_Visualizer/ArtReview_Visualizer
   bIsVariable=true
   parent=VerticalBox_51
   widget=ArtReview_PostProcess
   slot {class=VerticalBoxSlot}
  }
  VerticalBox_132 {
   widgetClassPath=&VerticalBox
   parent=HorizontalBox_60
   widget=VerticalBox_132
   slot {class=HorizontalBoxSlot, properties {size {sizeRule=Fill, value=1}}}
  }
  ArtReview_SunTool_RealSun {
   widgetClassPath=/Game/ArtTools/Tech_Art/ArtReview_SunTool/SupportWidgets/ArtReview_SunTool_RealSun
   bIsVariable=true
   parent=VerticalBox_132
   widget=ArtReview_SunTool_RealSun
   slot {class=VerticalBoxSlot, properties {size {sizeRule=Fill, value=1}}}
  }
  ArtReview_CameraControls_C_10 {
   widgetClassPath=/Game/ArtTools/Tech_Art/ArtReview_SunTool/SupportWidgets/ArtReview_CameraControls
   bIsVariable=true
   parent=VerticalBox_2
   widget=ArtReview_CameraControls_C_10
   slot {class=VerticalBoxSlot}
  }
 ]
}
